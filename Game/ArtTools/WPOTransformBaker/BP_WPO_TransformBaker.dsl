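; Editor tool actor that bakes a skeletal animation into a texture, apparently for World Position Offset crowd animation.
; AnimToRenderTarget samples at 30 fps, drawing frames as columns and socket position, quaternion and optional scale as row blocks, using bind-pose deltas.
; CreateProceduralSM copies a static mesh, encoding each vertex's nearest bone index in vertex color.
; Helpers export and reimport the target as EXR, draw the skeleton and scatter instanced crowd meshes.
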(event EventBeginPlay)

(event Collision|EventActorBeginOverlap
  (params (OtherActor :type Actor%20Object%20Reference)))

(event EventTick
  (params (DeltaSeconds :type Float%20%28single-precision%29)))

(event Custom|AnimToRenderTarget
  (bind _spawnedskeletalmesh (Variables|Default|GetSpawnedSkeletalMesh))
  (Rendering|ClearRenderTarget2D (Variables|Default|GetrenderTarget))
  ;[1 1 1] put sockets into an array so it's easy to check what sockets are used
  (Variables|Debug|Setsockets (Transformation|GetAllSocketNames (Variables|Default|GetSkeletalMeshComponent)))
  ;[1 1 1] create a bindPose skeletalMesh Component ... we will work out deltas against this
  (bind _returnvalue (AddComponent|NODE_AddSkeletalMeshComponent-1))
  (bind _output_get (Variables|Default|SetbindPoseSkeletalMeshComponent _returnvalue))
  (Components|SkinnedMesh|SetSkinnedAssetandUpdate _output_get (Variables|Default|GetskeletalMesh) true)
  ;[1 1 1] loop through all the frames
  (for _index (range (+ (Variables|RenderTargetInfo|GetxSize) 1))
    (Variables|Default|Setframe _index)
    ;[1 1 1] skeletal mesh component per frame
    (bind _returnvalue_2 (AddComponent|NODE_AddSkeletalMeshComponent-0))
    (bind _output_get_2 (Variables|Default|SetSpawnedSkeletalMesh _returnvalue_2))
    (Components|SkinnedMesh|SetSkinnedAssetandUpdate _output_get_2 (Variables|/Script/Engine.SkinnedMeshComponent|GetSkeletalMesh (Variables|Default|GetSkeletalMeshComponent)) true)
    (Components|SkeletalMesh|SetUpdateAnimationinEditor _spawnedskeletalmesh true)
    (Components|Animation|OverrideAnimationData _spawnedskeletalmesh (Variables|Default|GetanimSequence) false true (* (Math|Conversions|ToFloat(Integer) (Variables|Default|Getframe)) (/ 1.0 3e+01)) 1.0)
    (CallFunction|RenderAnimFrametoRenderTarget (Variables|Default|Getframe) (Variables|Default|GetSpawnedSkeletalMesh) (Variables|Default|GetbindPoseSkeletalMeshComponent) (== (Variables|Default|Getframe) 0))))

(event Custom|CreateProceduralSM
  (bind _skeletalmeshcomponent (Variables|Default|GetSkeletalMeshComponent))
  (bind _returnvalue (Transformation|GetAllSocketNames _skeletalmeshcomponent))
  (bind _numsockets (Variables|Debug|GetnumSockets))
  (bind _returnvalue_2 (Math|Conversions|ToFloat(Integer) _numsockets))
  (bind _skeletalmeshcomponent_2 (Variables|Default|GetSkeletalMeshComponent))
  (Components|SkinnedMesh|SetSkinnedAssetandUpdate _skeletalmeshcomponent_2 (Variables|Default|GetskeletalMesh) true)
  (Variables|Debug|SetnumSockets (Utilities|Array|Length (Transformation|GetAllSocketNames _skeletalmeshcomponent_2)))
  (Variables|Default|SetactorPosition (Transformation|GetActorLocation self))
  ;[1 1 1] build line arrays
  (Utilities|Array|Clear (Variables|Debug|Getlines_start))
  (Utilities|Array|Clear (Variables|Debug|Getlines_end))
  (Utilities|Array|Clear (Variables|Debug|Getlines_names))
  (Utilities|Array|Clear (Variables|Debug|Getlines_index))
  (for _array_element _returnvalue
    (if (and (Utilities|Name|NotEqual(Name) (Components|SkinnedMesh|GetParentBone _skeletalmeshcomponent _array_element) "Root") (Utilities|Name|NotEqual(Name) (Components|SkinnedMesh|GetParentBone _skeletalmeshcomponent _array_element) "None"))
      (Utilities|Array|Add (Variables|Debug|Getlines_start) (Transformation|GetSocketLocation _skeletalmeshcomponent (Components|SkinnedMesh|GetParentBone _skeletalmeshcomponent _array_element)))
      (Utilities|Array|Add (Variables|Debug|Getlines_end) (Transformation|GetSocketLocation _skeletalmeshcomponent _array_element))
      (Utilities|Array|Add (Variables|Debug|Getlines_names) (Utilities|String|StringToName (Utilities|String|Append "" (Utilities|String|ToString(Name) (Components|SkinnedMesh|GetParentBone _skeletalmeshcomponent _array_element)) "
           " (Utilities|String|ToString(Name) _array_element) " (" (Utilities|String|ToString(Integer) (Utilities|Array|FindItem _returnvalue (Components|SkinnedMesh|GetParentBone _skeletalmeshcomponent _array_element))) ")")))
      (Utilities|Array|Add (Variables|Debug|Getlines_index) (Utilities|Array|FindItem _returnvalue (Components|SkinnedMesh|GetParentBone _skeletalmeshcomponent _array_element)))))
  (for (_array_element_2 _array_index) (Variables|Debug|Getlines_start)
    (Development|PrintString (Utilities|String|Append (Utilities|String|ToString(Integer) _array_index) " " (Utilities|String|ToString(Vector) _array_element_2) "" (Utilities|String|ToString(Vector) (Utilities|Array|Get(acopy) (Variables|Debug|Getlines_end) _array_index)))))
  ;[1 1 1] get the mesh data from staticMesh
  (bind (_vertices _triangles _normals _uvs _tangents) (Components|ProceduralMesh|GetSectionfromStaticMesh (Variables|ProceduralSM|GetstaticMesh)))
  (Variables|Default|Setvertices _vertices)
  (Variables|Default|Settriangles _triangles)
  (Variables|Default|Setnormals _normals)
  (Variables|Default|Setuvs _uvs)
  (Variables|Default|Settangents _tangents)
  (Utilities|Array|Clear (Variables|Default|Getcolors))
  ;[1 1 1] loop through all vertices
  (for (_array_element_3 _array_index_2) (Variables|Default|Getvertices)
    ;[1 1 1] closest bone
    (bind (_bonename _boneindex) (CallFunction|FindClosestBoneLine (+ _array_element_3 (Transformation|GetActorLocation self)) (Variables|Default|GetSkeletalMeshComponent) (Variables|Debug|Getlines_start) (Variables|Debug|Getlines_end) (Variables|Debug|Getlines_names) (Variables|Debug|Getlines_index)))
    ;[1 1 1] vertex color
    (Utilities|Array|Add (Variables|Default|Getcolors) (Math|Color|MakeColor (/ (Math|Conversions|ToFloat(Integer) _boneindex) _returnvalue_2)))
    (Development|PrintString (Utilities|String|Append (Utilities|String|ToString(Integer) _array_index_2) " " (Utilities|String|ToString(Integer) _boneindex) " " (Utilities|String|ToString(Float) (/ (Math|Conversions|ToFloat(Integer) _boneindex) _returnvalue_2)) " " (Utilities|String|ToString(Name) _bonename) " " (Utilities|String|ToString(Vector) _array_element_3))))
  (bind _returnvalue_3 (AddComponent|NODE_AddProceduralMeshComponent-0))
  ;[1 1 1] create the new procedural mesh
  (Components|ProceduralMesh|CreateMeshSection :self _returnvalue_3 :SectionIndex 0 :Vertices (Variables|Default|Getvertices) :Triangles (Variables|Default|Gettriangles) :Normals (Variables|Default|Getnormals) :UV0 (Variables|Default|Getuvs) :VertexColors (Variables|Default|Getcolors) :Tangents (Variables|Default|Gettangents) :bCreateCollision false :bSRGBConversion false)
  (Rendering|Material|SetMaterial _returnvalue_3 0 (Variables|ProceduralSM|Getmaterial)))

(event Custom|DrawSkeleton
  (bind _skeletalmeshcomponent (Variables|Default|GetSkeletalMeshComponent))
  (Components|SkinnedMesh|SetSkinnedAssetandUpdate (Variables|Default|GetSkeletalMeshComponent) (Variables|Default|GetskeletalMesh) true)
  (for _array_element (Transformation|GetAllSocketNames _skeletalmeshcomponent)
    (if (Utilities|Name|NotEqual(Name) (Components|SkinnedMesh|GetParentBone _skeletalmeshcomponent _array_element) "None")
      (Rendering|Debug|DrawDebugLine (Transformation|GetSocketLocation _skeletalmeshcomponent (Components|SkinnedMesh|GetParentBone _skeletalmeshcomponent _array_element)) (Transformation|GetSocketLocation _skeletalmeshcomponent _array_element) "(R=1.000000,G=0.000000,B=0.127610,A=1.000000)" 1e+01 1.0))))

(fn BakeRenderTargetToStaticTexture
  (bind _rendertarget (Variables|Default|GetrenderTarget))
  (bind _returnvalue (Utilities|Paths|GetSystemPath _rendertarget))
  (bind _returnvalue_2 (Utilities|Paths|GetBaseFilename _returnvalue))
  (bind _rendertarget_2 (Variables|Default|GetrenderTarget))
  (bind _returnvalue_3 (Utilities|GetObjectPathString _rendertarget_2))
  (bind _returnvalue_4 (Utilities|Paths|GetPath _returnvalue_3))
  (bind _returnvalue_5 (Utilities|Paths|GetPath _returnvalue))
  (bind _returnvalue_6 (Utilities|String|Append _returnvalue_2 "_baked.exr"))
  (bind _returnvalue_7 (Utilities|String|Append _returnvalue_5 "/" _returnvalue_6))
  (Rendering|ExportRenderTarget (Variables|Default|GetrenderTarget) _returnvalue_5 _returnvalue_6)
  (Development|PrintString _returnvalue_7)
  (Development|PrintString _returnvalue_4)
  (Utilities|FlowControl|Sequence
    (:then_0
      (bind _returnvalue_8 (Game|ConstructObjectfromClass "/Script/UnrealEd.AutomatedAssetImportData" self)) ; Construct From Object Class Node
      (bind _output_get (Variables|Default|SetImportData _returnvalue_8))
      (Variables|/Script/UnrealEd.AutomatedAssetImportData|SetFilenames (Utilities|Array|MakeArray _returnvalue_7) _output_get)
      (Variables|/Script/UnrealEd.AutomatedAssetImportData|SetDestinationPath _returnvalue_4 _output_get)
      (Variables|/Script/UnrealEd.AutomatedAssetImportData|SetbReplaceExisting true _output_get))
    (:then_1
      (bind _returnvalue_9 (EditorScripting|AssetTools|GetAssetTools))
      (bind _returnvalue_10 (EditorScripting|AssetTools|ImportAssetsAutomated _returnvalue_9 _returnvalue_8))
      (for _array_element _returnvalue_10
        (Development|PrintString (Utilities|GetObjectPathString _array_element))))))

(fn CreateStaticMeshInstances
  (bind _returnvalue (AddComponent|NODE_AddHierarchicalInstancedStaticMeshComponent-0 :RelativeTransform_Rotation "0, 0, 0" :RelativeTransform_Scale "1.000000,1.000000,1.000000"))
  (Components|StaticMesh|SetStaticMesh _returnvalue (Variables|ProceduralSM|GetstaticMesh))
  (for _ (range (+ (Variables|Default|GetcrowdNum) 1))
    (Components|InstancedStaticMesh|AddInstance _returnvalue (Math|Transform|MakeTransform (Math|Vector|MakeVector (Math|Random|RandomFloatinRange -1e+03 1e+03) (Math|Random|RandomFloatinRange -1e+03 1e+03))) false)))

(fn DrawPixelToRenderTarget
  (params (x :type Integer) (y :type Integer) (color :type Linear%20Color%20Structure))
  (bind (_canvas _size _context) (Rendering|BeginDrawCanvastoRenderTarget (Variables|Default|GetrenderTarget)))
  (Canvas|DrawTexture _canvas "" (Math|Vector2D|MakeVector2D (Math|Conversions|ToFloat(Integer) x) (Math|Conversions|ToFloat(Integer) y)) "(X=1.000000,Y=1.000000)" "(X=0.000000,Y=0.000000)" "(X=1.000000,Y=1.000000)" color "BLEND_Opaque" 0.0 "(X=0.500000,Y=0.500000)")
  (Rendering|EndDrawCanvastoRenderTarget _context))

(fn FindClosestBoneLine
  (params (position :type Vector) (skeletalMeshComponent :type Skeletal%20Mesh%20Component%20Object%20Reference) (lines_start :type Array%20of%20Vectors) (lines_end :type Array%20of%20Vectors) (lines_names :type Array%20of%20Names) (lines_index :type Array%20of%20Integers))
  (bind _point (LocalVariables|Get_point))
  (LocalVariables|Set_distance 1e+03)
  (LocalVariables|Set_boneName "None")
  (LocalVariables|Set_index 0)
  (LocalVariables|Set_point "0, 0, 0")
  (for (_array_element _array_index) lines_start
    (if (< (Math|Vector|VectorLength (- position (Math|Vector|FindClosestPointonSegment position _array_element (Utilities|Array|Get(acopy) lines_end _array_index)))) (LocalVariables|Get_distance))
      (LocalVariables|Set_distance (Math|Vector|VectorLength (- position (Math|Vector|FindClosestPointonSegment position _array_element (Utilities|Array|Get(acopy) lines_end _array_index)))))
      (LocalVariables|Set_boneName (Utilities|Array|Get(acopy) lines_names _array_index))
      (LocalVariables|Set_index (Utilities|Array|Get(acopy) lines_index _array_index))
      (LocalVariables|Set_point (Math|Vector|FindClosestPointonSegment position _array_element (Utilities|Array|Get(acopy) lines_end _array_index)))))
  (Rendering|Debug|DrawDebugLine position _point "(R=1.000000,G=0.005135,B=0.000000,A=1.000000)" 5e+01 0.1)
  (return (LocalVariables|Get_boneName) (LocalVariables|Get_index) _point))

(fn RenderAnimFrameToRenderTarget
  (params (frame :type Integer) (skeletalMeshComponent :type Skeletal%20Mesh%20Component%20Object%20Reference) (bindPoseSkeletalMeshComponent :type Skeletal%20Mesh%20Component%20Object%20Reference) (useBindPose :type Boolean))
  (bind _localframe (LocalVariables|GetlocalFrame))
  (bind _yoffset (Variables|Debug|GetyOffset))
  (bind _yoffset_2 (Variables|Debug|GetyOffset))
  (LocalVariables|SetlocalFrame frame)
  ;[1 1 1] loop through all sockets
  (for (_array_element _array_index) (Variables|Debug|Getsockets)
    (if useBindPose
      ;[1 1 1] draw position
      (bind (_returnvalue_location _returnvalue_rotation) (Math|Transform|BreakTransform (Transformation|GetSocketTransform bindPoseSkeletalMeshComponent _array_element "RTS_Actor")))
      (CallFunction|DrawPixeltoRenderTarget (LocalVariables|GetlocalFrame) _array_index (Math|Conversions|ToLinearColor(Vector) _returnvalue_location))
      (else
        ;[1 1 1] draw position
        (bind (_returnvalue_location_2 _returnvalue_rotation_2 _returnvalue_scale) (Math|Transform|BreakTransform (Transformation|GetSocketTransform skeletalMeshComponent _array_element "RTS_Actor")))
        (bind (_returnvalue_scale_x _returnvalue_scale_y _returnvalue_scale_z) (Math|Vector|BreakVector _returnvalue_scale))
        (CallFunction|DrawPixeltoRenderTarget (LocalVariables|GetlocalFrame) _array_index (Math|Conversions|ToLinearColor(Vector) (- _returnvalue_location_2 _returnvalue_location)))))
    (bind (_returnvalue_roll _returnvalue_pitch _returnvalue_yaw) (Math|Rotator|BreakRotator (Math|Rotator|CombineRotators (Math|Rotator|InvertRotator _returnvalue_rotation) _returnvalue_rotation_2)))
    (bind _output_get (LocalVariables|SetlocalQuat (Math|Quat|MakefromEuler(Quat) (Math|Vector|MakeVector _returnvalue_roll _returnvalue_pitch _returnvalue_yaw))))
    (Math|Quat|EnsureShortestArcTo(Quat) _output_get (Math|Quat|MakefromEuler(Quat) (1.8e+02 0.0 0.0)))
    (bind _output_get_2 (Variables|Default|Setquat _output_get))
    (bind (_output_get_x _output_get_y _output_get_z _output_get_w) (Math|Quat|BreakQuat _output_get_2))
    ;[1 1 1] draw rotation
    (CallFunction|DrawPixeltoRenderTarget (LocalVariables|GetlocalFrame) (+ _array_index _yoffset) (Utilities|Struct|MakeLinearColor _output_get_x _output_get_y _output_get_z _output_get_w))
    ;[1 1 1] scale
    (if (Variables|Default|GetincludeScaleAnimation)
      (if useBindPose
        (CallFunction|DrawPixeltoRenderTarget _localframe (+ (+ _array_index _yoffset) _yoffset_2))
        (else
          (CallFunction|DrawPixeltoRenderTarget _localframe (+ (+ _array_index _yoffset) _yoffset_2) (Math|Conversions|ToLinearColor(Vector) (Math|Vector|MakeVector _returnvalue_scale_x _returnvalue_scale_z _returnvalue_scale_y))))))))

(fn ConstructionScript
  (bind _skeletalmeshcomponent (Variables|Default|GetSkeletalMeshComponent))
  (bind _returnvalue (Transformation|GetAllSocketNames _skeletalmeshcomponent))
  (Components|SkinnedMesh|SetSkinnedAssetandUpdate (Variables|Default|GetSkeletalMeshComponent) (Variables|Default|GetskeletalMesh) true)
  (Variables|Debug|Setsockets _returnvalue)
  (Variables|Debug|SetnumSockets (Utilities|Array|Length _returnvalue))
  (bind _output_get (Variables|Debug|SetanimPlayLength (Animation|GetPlayLength (Variables|Default|GetanimSequence))))
  (Variables|RenderTargetInfo|SetxSize (Math|Float|Truncate (* 30 _output_get)))
  (if (Variables|Default|GetincludeScaleAnimation)
    (bind _output_get_2 (Variables|RenderTargetInfo|SetySize (* (Variables|Debug|GetnumSockets) 3)))
    (Variables|Debug|SetyOffset (/ _output_get_2 3))
    (else
      (bind _output_get_3 (Variables|RenderTargetInfo|SetySize (* (Variables|Debug|GetnumSockets) 2)))
      (Variables|Debug|SetyOffset (/ _output_get_3 2)))))
