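# M_WPO_Transform is a Default Lit material in the WPOTransformBaker art tools, usable on instanced static, skeletal and deformer-driven meshes.
# It calls MF_WPO_Transform, by name a world position offset transform, leaving TimeInput unconnected.
# For debugging, static switches replace flat gray with vertex color shown as hue-shifted red.
# A socket view highlights one socket by sphere-masking vertex red against socketIndex divided by numSockets, dimming the rest.
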Material /Game/ArtTools/WPOTransformBaker/M_WPO_Transform
graph {
 properties {
  bTangentSpaceNormal=false
  bUsedWithInstancedStaticMeshes=true
  bUsedWithMeshDeformer=true
  bUsedWithSkeletalMesh=true
  pixelDepthOffsetMode=PDOM_Legacy
  usedShadingModels=Default Lit
 }
 nodes [
  Constant()
  MaterialFunctionCall!(TimeInput) {materialFunction=&/MF_WPO_Transform}
  n1 = ScalarParameter() {defaultValue=0.05, parameterName=A}
  n2 = MaterialFunctionCall(Hue Shift Percentage=VertexColor(), Texture=Constant3Vector() {constant=<rgba 1 0 0 0>}) {materialFunction=&HueShift}
  n3 = Divide(A=Constant() {r=1}, B=ScalarParameter() {defaultValue=10, group=Socket Visualization, parameterName=numSockets, sortPriority=21})
  n4 = SphereMask(
   A=VertexColor().R
   B=Multiply(A=ScalarParameter() {group=Socket Visualization, parameterName=socketIndex, sortPriority=25}, B=@n3)
   Radius=Multiply(A=@n3) {constB=0.5}
  ) {attenuationRadius=0.001}
  StaticSwitchParameter(
   True=StaticSwitchParameter(True=LinearInterpolate(A=@n1, B=@n2.Result, Alpha=@n4), False=@n2.Result) {group=Socket Visualization, parameterName=visualizeSocket, sortPriority=20}
   False=Constant() {r=0.2}
  ) {group=Socket Visualization, parameterName=visualizeWeights, sortPriority=10}
 ]
}
